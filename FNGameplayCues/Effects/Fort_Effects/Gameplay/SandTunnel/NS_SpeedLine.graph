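# Niagara system drawing short-lived white speed-line sprites around a moving owner, part of the SandTunnel gameplay effects.
# A single distance-culled emitter spawns by distance travelled in a raised sphere, giving particles reversed, scaled inherited velocity.
# Sprites grow over life and stretch with speed, while alpha follows a life curve scaled by User.SpeedLine and multiplied by System.PlayerSpeed.
# Rate spawning, downward velocity and particle-level player-speed velocity modules are disabled.
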NiagaraSystem /FNGameplayCues/Effects/Fort_Effects/Gameplay/SandTunnel/NS_SpeedLine
dependencies {
 usedDataInterfaces [&NiagaraDataInterfaceVector2DCurve, &NiagaraDataInterfaceCurve]
 usedDynamicInputs []
 usedModules [
  &SystemState
  &self:SystemSpawnScript.SystemScriptSource.SystemScriptGraph.NiagaraNodeAssignment_1.SetVariables_#1
 ]
}
emitters {Speedlines {
  emitterUpdateScript [
   &EmitterState(
    Delay First Loop Only=true
    Enable Distance Culling=true
    Inactive Response=::NewEnumerator1
    Life Cycle Mode=::NewEnumerator1
    Loop Delay=1
    Loop Duration=1
    MaxDistance=2048
    Module.bUseMaxDistance=true
    Module.bUseMinDistance=false
    Scalability Mode=::NewEnumerator1
    UseLoopDelay=true
   )
   &SpawnRate(SpawnRate=100, Use Spawn Probability=false) {enabled=false}
   &SpawnPerUnit(Module.Use Max Movement Threshold=true, Module.Use Movement Tolerance=false, Spawn Per Unit=4, Use Spawn Probability=false)
  ]
  particleSpawnScript [
   /Niagara/Modules/Spawn/Initialization/InitializeParticle(
    Color=&MakeLinearColorFromVectorAndFloat(
     "Vector (RGB)"=<1 1 1>
     "Float (Alpha)"=&Multiply_Float(A=&UniformRangedFloat(Minimum=0.2, Maximum=0.3), B=1)
    )
    Lifetime=&UniformRangedFloat(Minimum=0.1, Maximum=0.15)
    Module.Write Color=true
    Module.Write Lifetime=true
    Module.Write Mass=true
    Module.Write Material Random=true
    Module.Write Position=false
    Module.Write Scale=false
    Module.Write SpriteRotation=false
    Module.Write SpriteSize=true
    Position=&SimulationPosition()
    Sprite Size=&UniformRangedVector2D(Minimum=<4 16>, Maximum=<6 24>)
   )
   &SphereLocation(
    MaskSpawnGroup=false
    Offset=<0 0 200>
    Override Seed=false
    Sphere Radius=600
    Surface Only=false
    UseNonUniformScale=false
    UseOffset=true
    UseSphereOrientation=false
   )
   &AddVelocity(Velocity=<0 0 -100>) {enabled=false}
   /SetVariables(
    Particles.PlayerSpeed:NiagaraFloat=&NormalizeFloat(Value=&VectorLength(Vector=$Particles.PlayerVelocity), Min=0, Max=1)
    Particles.PlayerVelocity:Vector3f=$Engine.Owner.Velocity
   ) {enabled=false, isSetParameters=true}
   &AddVelocityFromPoint(
    Module.Use Falloff Distance=false
    Velocity Strength=&Multiply_Float(A=&UniformRangedFloat(Minimum=100, Maximum=200), B=$Particles.PlayerSpeed)
   ) {enabled=false}
   &InheritVelocity(Module.Limit Speed=true, Velocity Limit=500, Velocity Scale=<-0.25 -0.25 -0.25>)
  ]
  particleUpdateScript [
   &ParticleState()
   &SolveForcesAndVelocity(Clamp Velocity=false, Limit Acceleration=false)
   &ScaleSpriteSize(Scale Factor=&Vector2DFromCurve(
     Vector2Curve=DataInterface() {propertyValues {XCurve {keys [<at 0 0.5 RCIM_Cubic>,<at 0.5 1 RCIM_Cubic>]}, YCurve {keys [<at 0 0 RCIM_Cubic>,<at 0.5 1 RCIM_Cubic>]}}}
     ScaleCurve=&Vector2DFromFloat(Value=1)
    ))
   &ScaleSpriteSizeBySpeed(Max Scale Factor=<1 25>, Min Scale Factor=<0.01 0.05>, Velocity Threshold=1)
   &ScaleColor(
    Scale Alpha=&Multiply_Float(
     A=&FloatFromCurve(
      FloatCurve=DataInterface() {propertyValues {Curve {keys [
          <at 0 0 RCIM_Cubic>
          <at 0.3759 1.0207>
          {arriveTangent=-2.221, interpMode=RCIM_Cubic, leaveTangent=-2.221, tangentMode=RCTM_User, time=1, value=0}
         ]}}}
      Scale Curve=$User.SpeedLine
     )
     B=&FloatFromCurve(FloatCurve=DataInterface() {propertyValues {Curve {keys [<at 0 0>,<at 1 1>]}}}, CurveIndex=$System.PlayerSpeed)
    )
    Scale RGB=&VectorFromFloat(Value=1)
    ScaleA=true
    ScaleRGB=true
   )
  ]
  rendererClasses [&NiagaraSpriteRendererProperties]
  renderers [{rendererClass=&NiagaraSpriteRendererProperties}]
 }}
summary {
 systemName=NS_SpeedLine
 emitters [{emitterName=Speedlines, rendererClasses [&NiagaraSpriteRendererProperties]}]
 userVariables [User.SpeedLine {type=NiagaraFloat, defaultValue=1, description=""}]
}
